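# Animation Blueprint for the HighBeam glider, a bird-like skydiving chute with frontend preview animations.
# Locomotion moves through closed, opening and open glide, using aim-driven blendspaces from SkydivingState.
# Additive layers stack flap, lean, turn, wing, feather, look, blink and caw motion, several scaled by the inverted ParaGlide_Lean curve.
# Curve-driven Leg IK pins claws to handle virtual bones before DefaultSlot; Locomotion_Backup is left unconnected.
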AnimBlueprint /BRCosmetics/Animation/Game/MainPlayer/Skydive/ParaGlide/HighBeam/HighBeam_Glider_New_AnimBP
target_skeleton=/BRCosmetics/Gadgets/Assets/VinderTech_GliderChute/Glider_HighBeam/Meshes/HighBeam_Glider_Skeleton
graphs [
 AnimGraph {nodes [
   n1 = "Equal (Float)"!(A:double=GetCurveValue(CurveName=DisableFlap), B:double=1) -> (ReturnValue:bool) {is=title K2Node_CallFunction}
   n2 = Get CanFlap() -> (CanFlap:bool)
   n3 = Get TransitionFlap() -> (TransitionFlap:bool)
   Apply Additive!(
    Base
    Additive="Sequence Player (additive) 'HighBeam_Flap_Anim_GLIDER'"!(PlayRate:float=Get Flap_Rate() -> (Flap_Rate:double)) -> (Pose) {sequence=&/HighBeam_Flap_Anim_GLIDER}
    bAlphaBoolEnabled:bool=AND Boolean(A=OR Boolean(A=@n2, B=@n3), B=NOT Boolean(A=@n1))
   ) -> (Pose) {alphaInputType=Bool, alphaBoolBlend {blendOption=Linear}}
   AnimGraphNode_StateMachine_0 = Flap() {is=title AnimGraphNode_StateMachine}
   AnimGraphNode_StateMachine_4 = Locomotion() {is=title AnimGraphNode_StateMachine}
   n4 = Break Fort Anim Input Skydiving!(FortAnimInput_Skydiving:"Fort Anim Input Skydiving Structure (by ref)"=Get SkydivingState() -> (SkydivingState:Fort Anim Input Skydiving Structure)) -> (
    ParaGilderRootOffset:Vector=~
    ParaGliderRootRotationOverride:Rotator=~
    bUseParaGlideRootModifier:bool=~
    bIsSkydivingFromLaunchPad:bool=~
    bIsSkydivingFromBus:bool=~
    bIsInVortex:bool=~
    bIsDiving:bool=~
    bIsParachuteOpen:bool=~
    bIsSkydiving:bool=~
    bIsParachuteLeaning:bool=~
    bIsSkydiveLeaning:bool=~
    bIsLeaning:bool=~
    bIsSkydiveDiveMode:bool=~
    bPlaySkydiveDrift:bool=~
    bSkydiveDriftDelayActive:bool=~
    bSkydiveDriftAnimAllowed:bool=~
    bIsGliderRight:bool=~
    bIsGliderCenter:bool=~
    bIsGliderLeft:bool=~
    bIsGliderForward:bool=~
    bIsGliderBack:bool=~
    bHasDeployedAndIsStandingUpright:bool=~
    LocalAccelForward:float=~
    LocalAccelRight:float=~
    LocalVelocityRight:float=~
    SkydiveAimPitch:float=~
    SkydiveAimYaw:float=~
    DeployChuteAnimRate:float=~
    SkydiveDriftAnimRate:float=~
    SkydiveDriftAnimRateCurrent:float=~
    SkydiveFidgetAnimRate:float=~
    SkydiveFidgetAnimRateCurrent:float=~
    SkydiveAdditiveAlpha:float=~
    SkydiveDriftDelay:float=~
    SkydiveDriftAnim:int
    SkydiveDriftAnimMax:int
    LaunchpadAnim:int
    LocalAccelDir:ESkydivingDirection Enum
    DirectionLast:ESkydivingDirection Enum
    PoseAdditive:Anim Sequence Object Reference
   ) {is=title K2Node_BreakStruct}
   n5 = Break Fort Anim Input Skydiving(FortAnimInput_Skydiving=Get SkydivingState()) -> (
    ParaGilderRootOffset=~
    ParaGliderRootRotationOverride=~
    bUseParaGlideRootModifier=~
    bIsSkydivingFromLaunchPad=~
    bIsSkydivingFromBus=~
    bIsInVortex=~
    bIsDiving=~
    bIsParachuteOpen=~
    bIsSkydiving=~
    bIsParachuteLeaning=~
    bIsSkydiveLeaning=~
    bIsLeaning=~
    bIsSkydiveDiveMode=~
    bPlaySkydiveDrift=~
    bSkydiveDriftDelayActive=~
    bSkydiveDriftAnimAllowed=~
    bIsGliderRight=~
    bIsGliderCenter=~
    bIsGliderLeft=~
    bIsGliderForward=~
    bIsGliderBack=~
    bHasDeployedAndIsStandingUpright=~
    LocalAccelForward=~
    LocalAccelRight=~
    LocalVelocityRight=~
    SkydiveAimPitch=~
    SkydiveAimYaw=~
    DeployChuteAnimRate=~
    SkydiveDriftAnimRate=~
    SkydiveDriftAnimRateCurrent=~
    SkydiveFidgetAnimRate=~
    SkydiveFidgetAnimRateCurrent=~
    SkydiveAdditiveAlpha=~
    SkydiveDriftDelay=~
   )
   n6 = Apply Additive!(
    Base=Apply Additive!(
     Base=Apply Additive!(Base=@AnimGraphNode_StateMachine_4, Additive=@AnimGraphNode_StateMachine_0, Alpha:float) -> (Pose) {alphaBoolBlend {blendOption=Linear}}
     Additive=Blendspace Player 'BS_HighBeam_Lean_GLIDER'(X=@n4.LocalAccelRight, Y=@n4.LocalAccelForward) {blendSpace=&/BS_HighBeam_Lean_GLIDER}
     AlphaCurveName:Name=ParaGlide_Lean
    ) -> (Pose) {alphaInputType=Curve, alphaBoolBlend {blendOption=Linear}, alphaScaleBiasClamp {bias=1, scale=-1}}
    Additive=Blendspace Player 'BS_HighBeam_Turn_GLIDER'(X=@n5.SkydiveAimYaw, Y=@n5.SkydiveAimPitch) {blendSpace=&/BS_HighBeam_Turn_GLIDER}
    Alpha:float
   ) -> (Pose) {alphaBoolBlend {blendOption=Linear}}
   n7 = "Sequence Player (additive) 'HighBeam_WingTail_Adjust_GLIDER'"() {bLoopAnimation=false, sequence=&/HighBeam_WingTail_Adjust_GLIDER}
   AnimGraphNode_StateMachine_1 = LookStateMachine() {is=title AnimGraphNode_StateMachine}
   n8 = Apply Additive!(
    Base=Apply Additive(
     Base=Apply Additive(Base=@n6, Additive=@n7, AlphaCurveName=ParaGlide_Lean) {alphaInputType=Curve, alphaBoolBlend {blendOption=Linear}, alphaScaleBiasClamp {bias=1, scale=-1}}
     Additive="Sequence Player (additive) 'HighBeam_Frontend_GlideAdditive_Anim_GLIDER'"() {sequence=&/HighBeam_Frontend_GlideAdditive_Anim_GLIDER}
     AlphaCurveName=FrontendGlideAdditive
    ) {alphaInputType=Curve, alphaBoolBlend {blendOption=Linear}}
    Additive=@AnimGraphNode_StateMachine_1
    AlphaCurveName:Name=ParaGlide_Lean
   ) -> (Pose) {alphaInputType=Curve, alphaBoolBlend {blendOption=Linear}, alphaScaleBiasClamp {bias=1, scale=-1}}
   n9 = "Sequence Player (additive) 'HighBeam_Feathers_Anim_GLIDER'"() {sequence=&/HighBeam_Feathers_Anim_GLIDER}
   AnimGraphNode_StateMachine_2 = Blink() {is=title AnimGraphNode_StateMachine}
   n10 = Apply Additive!(
    Base=Apply Additive!(
     Base=Apply Additive!(Base=@n8, Additive=@n9, Alpha:float) -> (Pose) {alphaBoolBlend {blendOption=Linear}}
     Additive="Sequence Player (additive) 'HighBeam_WingSway_Anim_GLIDER'"() {sequence=&/HighBeam_WingSway_Anim_GLIDER}
     AlphaCurveName:Name=ParaGlide_Lean
    ) -> (Pose) {alphaInputType=Curve, alphaBoolBlend {blendOption=Linear}, alphaScaleBiasClamp {bias=1, scale=-1}}
    Additive=@AnimGraphNode_StateMachine_2
    Alpha:float
   ) -> (Pose) {alphaBoolBlend {blendOption=Linear}}
   AnimGraphNode_StateMachine_3 = Caw() {is=title AnimGraphNode_StateMachine}
   n11 = Leg IK!(
    ComponentPose:Component Space Pose Link Structure=Local To Component(LocalPose=Apply Additive(Base=@n10, Additive=@AnimGraphNode_StateMachine_3) {alphaBoolBlend {blendOption=Linear}})
    AlphaCurveName:Name=EnableIK
   ) -> (Pose:Component Space Pose Link Structure) {
    is=title AnimGraphNode_LegIK
    alphaInputType=Curve
    alphaBoolBlend {blendOption=Linear}
    legsDefinition [
     {
      bEnableKneeTwistCorrection=false
      bEnableRotationLimit=false
      footBoneForwardAxis=Y
      hingeRotationAxis=None
      minRotationAngle=15
      numBonesInLimb=2
      twistOffsetCurveName=None
      fKFootBone {boneName=JNT_L_Claw}
      iKFootBone {boneName=VB Handles_Base_JNT_L_Claw}
     }
     {
      bEnableKneeTwistCorrection=false
      bEnableRotationLimit=false
      footBoneForwardAxis=Y
      hingeRotationAxis=None
      minRotationAngle=15
      numBonesInLimb=2
      twistOffsetCurveName=None
      fKFootBone {boneName=JNT_R_Claw}
      iKFootBone {boneName=VB Handles_Base_JNT_R_Claw}
     }
    ]
   }
   Output Pose(Result=Slot 'DefaultSlot'(Source=Component To Local(ComponentPose=@n11))) {layerGroup=None}
   AnimGraphNode_StateMachine_5 = Locomotion_Backup() {is=title AnimGraphNode_StateMachine}
  ]}
 Flap {
  entry=Identity
  path=AnimGraph.AnimGraphNode_StateMachine_0.Flap
  ids=id
  nodes [
   AnimStateNode_0 = Identity() {is=title AnimStateNode}
   AnimStateNode_1 = Flap() {
    is=title AnimStateNode
    bAlwaysResetOnEntry=true
    stateEntered {guid=#1, notifyName=EnteredFlap}
    stateLeft {guid=#2, notifyName=ExitedFlap}
   }
   AnimStateTransitionNode_0 = AnimStateTransitionNode() {crossfadeDuration=0.6}
   AnimStateTransitionNode_1 = AnimStateTransitionNode() {bAutomaticRuleBasedOnSequencePlayerInState=true, crossfadeDuration=0.6}
  ]
  links [
   {dst=Identity, rule=AnimGraph.AnimGraphNode_StateMachine_0.Flap.AnimStateTransitionNode_1.Transition, src=Flap, via=AnimStateTransitionNode_1}
   {dst=Flap, rule=AnimGraph.AnimGraphNode_StateMachine_0.Flap.AnimStateTransitionNode_0.Transition, src=Identity, via=AnimStateTransitionNode_0}
  ]
 }
 Identity {path=AnimGraph.AnimGraphNode_StateMachine_0.Flap.AnimStateNode_0.Identity, nodes [Output Animation Pose(Result=Additive Identity Pose())]}
 Flap {
  path=AnimGraph.AnimGraphNode_StateMachine_0.Flap.AnimStateNode_1.Flap
  nodes [
   Output Animation Pose(Result="Sequence Player (additive) 'HighBeam_Flap_Anim_GLIDER'"(PlayRate=1) {bLoopAnimation=false, sequence=&HighBeam_Flap_Anim_GLIDER})
   Get Flap_Rate()
  ]
 }
 Transition {
  path=AnimGraph.AnimGraphNode_StateMachine_0.Flap.AnimStateTransitionNode_0.Transition
  nodes [
   n1 = "Equal (Float)"(A=GetCurveValue(CurveName=DisableFlap), B=1)
   n2 = Get CanFlap()
   n3 = Get TransitionFlap()
   Result(bCanEnterTransition=AND Boolean(A=OR Boolean(A=@n2, B=@n3), B=NOT Boolean(A=@n1)))
  ]
 }
 Transition {path=AnimGraph.AnimGraphNode_StateMachine_0.Flap.AnimStateTransitionNode_1.Transition, nodes [Result()]}
 LookStateMachine {
  entry=Identity
  path=AnimGraph.AnimGraphNode_StateMachine_1.LookStateMachine
  ids=id
  nodes [
   AnimStateNode_0 = Look Left() {is=title AnimStateNode, stateEntered {guid=#3, notifyName=Entered Look Left}}
   AnimStateNode_1 = Look Right() {is=title AnimStateNode, stateEntered {guid=#4, notifyName=Entered Look Right}}
   AnimStateNode_2 = Identity()
   AnimStateTransitionNode_2 = AnimStateTransitionNode() {crossfadeDuration=0.5}
   AnimStateTransitionNode_3 = AnimStateTransitionNode() {crossfadeDuration=0.5}
   AnimStateTransitionNode_4 = AnimStateTransitionNode() {bAutomaticRuleBasedOnSequencePlayerInState=true, crossfadeDuration=0.5}
   AnimStateTransitionNode_5 = AnimStateTransitionNode() {bAutomaticRuleBasedOnSequencePlayerInState=true, crossfadeDuration=0.5}
  ]
  links [
   {
    dst=Look Left
    rule=AnimGraph.AnimGraphNode_StateMachine_1.LookStateMachine.AnimStateTransitionNode_2.Transition
    src=Identity
    via=AnimStateTransitionNode_2
   }
   {
    dst=Look Right
    rule=AnimGraph.AnimGraphNode_StateMachine_1.LookStateMachine.AnimStateTransitionNode_3.Transition
    src=Identity
    via=AnimStateTransitionNode_3
   }
   {
    dst=Identity
    rule=AnimGraph.AnimGraphNode_StateMachine_1.LookStateMachine.AnimStateTransitionNode_4.Transition
    src=Look Left
    via=AnimStateTransitionNode_4
   }
   {
    dst=Identity
    rule=AnimGraph.AnimGraphNode_StateMachine_1.LookStateMachine.AnimStateTransitionNode_5.Transition
    src=Look Right
    via=AnimStateTransitionNode_5
   }
  ]
 }
 Look Left {
  path=AnimGraph.AnimGraphNode_StateMachine_1.LookStateMachine.AnimStateNode_0.Look Left
  nodes [
   n1 = "Sequence Player (additive) 'HighBeam_Head_Darts_Left_Anim_GLIDER'"() {bLoopAnimation=false, sequence=&/HighBeam_Head_Darts_Left_Anim_GLIDER}
   Output Animation Pose(Result=@n1)
  ]
 }
 Look Right {
  path=AnimGraph.AnimGraphNode_StateMachine_1.LookStateMachine.AnimStateNode_1.Look Right
  nodes [
   n1 = "Sequence Player (additive) 'HighBeam_Head_Darts_Right_Anim_GLIDER'"() {bLoopAnimation=false, sequence=&/HighBeam_Head_Darts_Right_Anim_GLIDER}
   Output Animation Pose(Result=@n1)
  ]
 }
 Identity {
  path=AnimGraph.AnimGraphNode_StateMachine_1.LookStateMachine.AnimStateNode_2.Identity
  nodes [Output Animation Pose(Result=Additive Identity Pose())]
 }
 Transition {
  path=AnimGraph.AnimGraphNode_StateMachine_1.LookStateMachine.AnimStateTransitionNode_2.Transition
  nodes [Result(bCanEnterTransition=Get CanLookLeft() -> (CanLookLeft:bool))]
 }
 Transition {
  path=AnimGraph.AnimGraphNode_StateMachine_1.LookStateMachine.AnimStateTransitionNode_3.Transition
  nodes [Result(bCanEnterTransition=Get CanLookRight() -> (CanLookRight:bool))]
 }
 Transition {path=AnimGraph.AnimGraphNode_StateMachine_1.LookStateMachine.AnimStateTransitionNode_4.Transition, nodes [Result()]}
 Transition {path=AnimGraph.AnimGraphNode_StateMachine_1.LookStateMachine.AnimStateTransitionNode_5.Transition, nodes [Result()]}
 Blink {
  entry=Identity
  path=AnimGraph.AnimGraphNode_StateMachine_2.Blink
  ids=id
  nodes [
   AnimStateNode_0 = Blink() {is=title AnimStateNode, stateLeft {guid=#5, notifyName=ExitedBlink}}
   AnimStateNode_1 = BlinkWithEyeDart() {is=title AnimStateNode, stateLeft {guid=#6, notifyName=ExitedBlinkWithEyeDart}}
   AnimStateNode_2 = Identity() {stateLeft {guid=#7, notifyName=ExitedBlinkIdentity}}
   AnimStateTransitionNode_1 = AnimStateTransitionNode() {crossfadeDuration=0}
   AnimStateTransitionNode_2 = AnimStateTransitionNode() {crossfadeDuration=0}
   AnimStateTransitionNode_3 = AnimStateTransitionNode() {bAutomaticRuleBasedOnSequencePlayerInState=true, crossfadeDuration=0}
   AnimStateTransitionNode_4 = AnimStateTransitionNode() {bAutomaticRuleBasedOnSequencePlayerInState=true, crossfadeDuration=0}
  ]
  links [
   {
    dst=Identity
    rule=AnimGraph.AnimGraphNode_StateMachine_2.Blink.AnimStateTransitionNode_4.Transition
    src=Blink
    via=AnimStateTransitionNode_4
   }
   {
    dst=Identity
    rule=AnimGraph.AnimGraphNode_StateMachine_2.Blink.AnimStateTransitionNode_3.Transition
    src=BlinkWithEyeDart
    via=AnimStateTransitionNode_3
   }
   {
    dst=Blink
    rule=AnimGraph.AnimGraphNode_StateMachine_2.Blink.AnimStateTransitionNode_2.Transition
    src=Identity
    via=AnimStateTransitionNode_2
   }
   {
    dst=BlinkWithEyeDart
    rule=AnimGraph.AnimGraphNode_StateMachine_2.Blink.AnimStateTransitionNode_1.Transition
    src=Identity
    via=AnimStateTransitionNode_1
   }
  ]
 }
 Blink {
  path=AnimGraph.AnimGraphNode_StateMachine_2.Blink.AnimStateNode_0.Blink
  nodes [Output Animation Pose(Result="Sequence Player (additive) 'HighBeam_Blink_Anim_GLIDER'"() {bLoopAnimation=false, sequence=&/HighBeam_Blink_Anim_GLIDER})]
 }
 BlinkWithEyeDart {
  path=AnimGraph.AnimGraphNode_StateMachine_2.Blink.AnimStateNode_1.BlinkWithEyeDart
  nodes [Output Animation Pose(Result="Sequence Player (additive) 'HighBeam_BlinkDart_Anim_GLIDER'"() {bLoopAnimation=false, sequence=&/HighBeam_BlinkDart_Anim_GLIDER})]
 }
 Identity {path=AnimGraph.AnimGraphNode_StateMachine_2.Blink.AnimStateNode_2.Identity, nodes [Output Animation Pose(Result=Additive Identity Pose())]}
 Transition {
  path=AnimGraph.AnimGraphNode_StateMachine_2.Blink.AnimStateTransitionNode_1.Transition
  nodes [Result(bCanEnterTransition=Get CanBlinkWithEyeDart() -> (CanBlinkWithEyeDart:bool))]
 }
 Transition {
  path=AnimGraph.AnimGraphNode_StateMachine_2.Blink.AnimStateTransitionNode_2.Transition
  nodes [Result(bCanEnterTransition=Get CanBlink() -> (CanBlink:bool))]
 }
 Transition {path=AnimGraph.AnimGraphNode_StateMachine_2.Blink.AnimStateTransitionNode_3.Transition, nodes [Result()]}
 Transition {path=AnimGraph.AnimGraphNode_StateMachine_2.Blink.AnimStateTransitionNode_4.Transition, nodes [Result()]}
 Caw {
  entry=Identity
  path=AnimGraph.AnimGraphNode_StateMachine_3.Caw
  ids=id
  nodes [
   AnimStateNode_0 = Identity() {stateLeft {guid=#8, notifyName=ExitedCawIdentity}}
   AnimStateNode_1 = Caw() {is=title AnimStateNode}
   AnimStateTransitionNode_5 = AnimStateTransitionNode() {blendMode=Sinusoidal}
   AnimStateTransitionNode_6 = AnimStateTransitionNode() {bAutomaticRuleBasedOnSequencePlayerInState=true, blendMode=Sinusoidal}
  ]
  links [
   {dst=Identity, rule=AnimGraph.AnimGraphNode_StateMachine_3.Caw.AnimStateTransitionNode_6.Transition, src=Caw, via=AnimStateTransitionNode_6}
   {dst=Caw, rule=AnimGraph.AnimGraphNode_StateMachine_3.Caw.AnimStateTransitionNode_5.Transition, src=Identity, via=AnimStateTransitionNode_5}
  ]
 }
 Identity {path=AnimGraph.AnimGraphNode_StateMachine_3.Caw.AnimStateNode_0.Identity, nodes [Output Animation Pose(Result=Additive Identity Pose())]}
 Caw {
  path=AnimGraph.AnimGraphNode_StateMachine_3.Caw.AnimStateNode_1.Caw
  nodes [Output Animation Pose(Result="Sequence Player (additive) 'HighBeam_Caw_Anim_GLIDER'"() {bLoopAnimation=false, sequence=&/HighBeam_Caw_Anim_GLIDER})]
 }
 Transition {
  path=AnimGraph.AnimGraphNode_StateMachine_3.Caw.AnimStateTransitionNode_5.Transition
  nodes [Result(bCanEnterTransition=Get CanCaw() -> (CanCaw:bool))]
 }
 Transition {path=AnimGraph.AnimGraphNode_StateMachine_3.Caw.AnimStateTransitionNode_6.Transition, nodes [Result()]}
 Locomotion {
  entry=Closed
  path=AnimGraph.AnimGraphNode_StateMachine_4.Locomotion
  ids=id
  nodes [
   AnimStateNode_0 = Idle_Open() {is=title AnimStateNode}
   AnimStateNode_1 = Open_Into() {is=title AnimStateNode}
   AnimStateNode_2 = Closed() {is=title AnimStateNode}
   AnimStateNode_3 = Flap()
   AnimStateTransitionNode_0 = AnimStateTransitionNode() {bAutomaticRuleBasedOnSequencePlayerInState=true, crossfadeDuration=0.3}
   AnimStateTransitionNode_1 = AnimStateTransitionNode() {crossfadeDuration=0}
   AnimStateTransitionNode_2 = AnimStateTransitionNode() {crossfadeDuration=0}
   AnimStateTransitionNode_3 = AnimStateTransitionNode() {crossfadeDuration=0, priorityOrder=2}
  ]
  links [
   {
    dst=Open_Into
    rule=AnimGraph.AnimGraphNode_StateMachine_4.Locomotion.AnimStateTransitionNode_1.Transition
    src=Closed
    via=AnimStateTransitionNode_1
   }
   {
    dst=Closed
    rule=AnimGraph.AnimGraphNode_StateMachine_4.Locomotion.AnimStateTransitionNode_2.Transition
    src=Idle_Open
    via=AnimStateTransitionNode_2
   }
   {
    dst=Closed
    rule=AnimGraph.AnimGraphNode_StateMachine_4.Locomotion.AnimStateTransitionNode_3.Transition
    src=Open_Into
    via=AnimStateTransitionNode_3
   }
   {
    dst=Idle_Open
    rule=AnimGraph.AnimGraphNode_StateMachine_4.Locomotion.AnimStateTransitionNode_0.Transition
    src=Open_Into
    via=AnimStateTransitionNode_0
   }
  ]
 }
 Idle_Open {
  path=AnimGraph.AnimGraphNode_StateMachine_4.Locomotion.AnimStateNode_0.Idle_Open
  nodes [
   n1 = Break Fort Anim Input Skydiving(FortAnimInput_Skydiving=Get SkydivingState()) -> (
    ParaGilderRootOffset=~
    ParaGliderRootRotationOverride=~
    bUseParaGlideRootModifier=~
    bIsSkydivingFromLaunchPad=~
    bIsSkydivingFromBus=~
    bIsInVortex=~
    bIsDiving=~
    bIsParachuteOpen=~
    bIsSkydiving=~
    bIsParachuteLeaning=~
    bIsSkydiveLeaning=~
    bIsLeaning=~
    bIsSkydiveDiveMode=~
    bPlaySkydiveDrift=~
    bSkydiveDriftDelayActive=~
    bSkydiveDriftAnimAllowed=~
    bIsGliderRight=~
    bIsGliderCenter=~
    bIsGliderLeft=~
    bIsGliderForward=~
    bIsGliderBack=~
    bHasDeployedAndIsStandingUpright=~
    LocalAccelForward=~
    LocalAccelRight=~
    LocalVelocityRight=~
    SkydiveAimPitch=~
    SkydiveAimYaw=~
    DeployChuteAnimRate=~
    SkydiveDriftAnimRate=~
    SkydiveDriftAnimRateCurrent=~
    SkydiveFidgetAnimRate=~
    SkydiveFidgetAnimRateCurrent=~
    SkydiveAdditiveAlpha=~
    SkydiveDriftDelay=~
   )
   n2 = Blendspace Player 'BS_HighBeam_GLIDER'(X=@n1.SkydiveAimYaw, Y=@n1.SkydiveAimPitch) {blendSpace=&/BS_HighBeam_GLIDER, groupName=Glide, method=SyncGroup}
   n3 = Get OwnerGlider() -> (OwnerGlider:Fort Player Parachute Object Reference)
   Output Animation Pose(Result=Blend Poses by bool(
     BlendPose_0=Sequence Player 'HighBeam_FrontEnd_Glide_Anim_GLIDER'() {sequence=&/HighBeam_FrontEnd_Glide_Anim_GLIDER}
     BlendPose_1=@n2
     bActiveValue=Get IsCosmeticPreview!(self:Fort Player Parachute Object Reference=@n3) -> (IsCosmeticPreview:bool)
    ) {blendType=Linear})
  ]
 }
 Open_Into {
  path=AnimGraph.AnimGraphNode_StateMachine_4.Locomotion.AnimStateNode_1.Open_Into
  nodes [
   n1 = Break Fort Anim Input Skydiving(FortAnimInput_Skydiving=Get SkydivingState()) -> (
    ParaGilderRootOffset=~
    ParaGliderRootRotationOverride=~
    bUseParaGlideRootModifier=~
    bIsSkydivingFromLaunchPad=~
    bIsSkydivingFromBus=~
    bIsInVortex=~
    bIsDiving=~
    bIsParachuteOpen=~
    bIsSkydiving=~
    bIsParachuteLeaning=~
    bIsSkydiveLeaning=~
    bIsLeaning=~
    bIsSkydiveDiveMode=~
    bPlaySkydiveDrift=~
    bSkydiveDriftDelayActive=~
    bSkydiveDriftAnimAllowed=~
    bIsGliderRight=~
    bIsGliderCenter=~
    bIsGliderLeft=~
    bIsGliderForward=~
    bIsGliderBack=~
    bHasDeployedAndIsStandingUpright=~
    LocalAccelForward=~
    LocalAccelRight=~
    LocalVelocityRight=~
    SkydiveAimPitch=~
    SkydiveAimYaw=~
    DeployChuteAnimRate=~
    SkydiveDriftAnimRate=~
    SkydiveDriftAnimRateCurrent=~
    SkydiveFidgetAnimRate=~
    SkydiveFidgetAnimRateCurrent=~
    SkydiveAdditiveAlpha=~
    SkydiveDriftDelay=~
   )
   n2 = Get OwnerGlider()
   Output Animation Pose(Result=Apply Additive!(
     Base=Blendspace Player 'BS_HighBeam_Into_GLIDER'(X=@n1.SkydiveAimYaw, Y=@n1.SkydiveAimPitch) {bLoop=false, blendSpace=&/BS_HighBeam_Into_GLIDER}
     Additive="Sequence Player (additive) 'HighBeam_FrontEnd_Into_Anim_GLIDER'"() {bLoopAnimation=false, sequence=&/HighBeam_FrontEnd_Into_Anim_GLIDER}
     bAlphaBoolEnabled:bool=Get IsCosmeticPreview(self=@n2)
    ) -> (Pose) {alphaInputType=Bool, alphaBoolBlend {blendOption=Linear}})
  ]
 }
 Closed {
  path=AnimGraph.AnimGraphNode_StateMachine_4.Locomotion.AnimStateNode_2.Closed
  nodes [
   n1 = Get OwnerGlider()
   Output Animation Pose(Result=Blend Poses by bool(
     BlendPose_0=Sequence Player 'HighBeam_Closed_Pose_GLIDER'() {sequence=&/HighBeam_Closed_Pose_GLIDER}
     BlendPose_1=Sequence Player 'HighBeam_Closed_Pose_GLIDER'() {sequence=&HighBeam_Closed_Pose_GLIDER}
     bActiveValue=Get IsCosmeticPreview(self=@n1)
    ) {blendType=Linear})
  ]
 }
 Flap {
  path=AnimGraph.AnimGraphNode_StateMachine_4.Locomotion.AnimStateNode_3.Flap
  nodes [
   "Sequence Player (additive) 'HighBeam_Flap_Anim_GLIDER'"(PlayRate=Get Flap_Rate()) {sequence=&HighBeam_Flap_Anim_GLIDER}
   n1 = Break Fort Anim Input Skydiving(FortAnimInput_Skydiving=Get SkydivingState()) -> (
    ParaGilderRootOffset=~
    ParaGliderRootRotationOverride=~
    bUseParaGlideRootModifier=~
    bIsSkydivingFromLaunchPad=~
    bIsSkydivingFromBus=~
    bIsInVortex=~
    bIsDiving=~
    bIsParachuteOpen=~
    bIsSkydiving=~
    bIsParachuteLeaning=~
    bIsSkydiveLeaning=~
    bIsLeaning=~
    bIsSkydiveDiveMode=~
    bPlaySkydiveDrift=~
    bSkydiveDriftDelayActive=~
    bSkydiveDriftAnimAllowed=~
    bIsGliderRight=~
    bIsGliderCenter=~
    bIsGliderLeft=~
    bIsGliderForward=~
    bIsGliderBack=~
    bHasDeployedAndIsStandingUpright=~
    LocalAccelForward=~
    LocalAccelRight=~
    LocalVelocityRight=~
    SkydiveAimPitch=~
    SkydiveAimYaw=~
    DeployChuteAnimRate=~
    SkydiveDriftAnimRate=~
    SkydiveDriftAnimRateCurrent=~
    SkydiveFidgetAnimRate=~
    SkydiveFidgetAnimRateCurrent=~
    SkydiveAdditiveAlpha=~
    SkydiveDriftDelay=~
   )
   n2 = Blendspace Player 'BS_HighBeam_GLIDER'(X=@n1.SkydiveAimYaw, Y=@n1.SkydiveAimPitch) {
    bIgnoreForRelevancyTest=true
    blendSpace=&BS_HighBeam_GLIDER
    groupName=Glide
    method=SyncGroup
   }
   n3 = Sequence Player 'HighBeam_FrontEnd_Glide_Anim_GLIDER'() {sequence=&HighBeam_FrontEnd_Glide_Anim_GLIDER}
   n4 = Get IsCosmeticPreview(self=Get OwnerGlider())
   AnimGraphNode_StateMachine_0 = New State Machine() {is=title AnimGraphNode_StateMachine}
   Output Animation Pose(Result=Apply Additive!(
     Base=Blend Poses by bool(BlendPose_0=@n3, BlendPose_1=@n2, bActiveValue=@n4) {blendType=Linear}
     Additive=@AnimGraphNode_StateMachine_0
     Alpha:float
    ) -> (Pose) {alphaBoolBlend {blendOption=Linear}})
   n5 = GetCurveValue(CurveName=DisableFlap)
   AND Boolean(A=OR Boolean(A=Get CanFlap(), B=Get TransitionFlap()), B=NOT Boolean(A="Equal (Float)"(A=@n5, B=1)))
  ]
 }
 New State Machine {
  entry=Identity
  path=AnimGraph.AnimGraphNode_StateMachine_4.Locomotion.AnimStateNode_3.Flap.AnimGraphNode_StateMachine_0.New State Machine
  ids=id
  nodes [
   AnimStateNode_0 = Identity()
   AnimStateNode_1 = Flap() {
    bAlwaysResetOnEntry=true
    stateEntered {guid=#1, notifyName=EnteredFlap}
    stateLeft {guid=#2, notifyName=ExitedFlap}
   }
   AnimStateTransitionNode_0 = AnimStateTransitionNode() {crossfadeDuration=0.6}
   AnimStateTransitionNode_1 = AnimStateTransitionNode() {bAutomaticRuleBasedOnSequencePlayerInState=true, crossfadeDuration=0.6}
  ]
  links [
   {
    dst=Identity
    rule=AnimGraph.AnimGraphNode_StateMachine_4.Locomotion.AnimStateNode_3.Flap.AnimGraphNode_StateMachine_0.New State Machine.AnimStateTransitionNode_1.Transition
    src=Flap
    via=AnimStateTransitionNode_1
   }
   {
    dst=Flap
    rule=AnimGraph.AnimGraphNode_StateMachine_4.Locomotion.AnimStateNode_3.Flap.AnimGraphNode_StateMachine_0.New State Machine.AnimStateTransitionNode_0.Transition
    src=Identity
    via=AnimStateTransitionNode_0
   }
  ]
 }
 Identity {
  path=AnimGraph.AnimGraphNode_StateMachine_4.Locomotion.AnimStateNode_3.Flap.AnimGraphNode_StateMachine_0.New State Machine.AnimStateNode_0.Identity
  nodes [Output Animation Pose(Result=Additive Identity Pose())]
 }
 Flap {
  path=AnimGraph.AnimGraphNode_StateMachine_4.Locomotion.AnimStateNode_3.Flap.AnimGraphNode_StateMachine_0.New State Machine.AnimStateNode_1.Flap
  nodes [
   Output Animation Pose(Result="Sequence Player (additive) 'HighBeam_Flap_Anim_GLIDER'"(PlayRate=1) {bLoopAnimation=false, sequence=&HighBeam_Flap_Anim_GLIDER})
   Get Flap_Rate()
  ]
 }
 Transition {
  path=AnimGraph.AnimGraphNode_StateMachine_4.Locomotion.AnimStateNode_3.Flap.AnimGraphNode_StateMachine_0.New State Machine.AnimStateTransitionNode_0.Transition
  nodes [
   n1 = Get CanFlap()
   Result(bCanEnterTransition=@n1)
   n2 = GetCurveValue(CurveName=DisableFlap)
   AND Boolean(A=OR Boolean(A=@n1, B=Get TransitionFlap()), B=NOT Boolean(A="Equal (Float)"(A=@n2, B=1)))
  ]
 }
 Transition {
  path=AnimGraph.AnimGraphNode_StateMachine_4.Locomotion.AnimStateNode_3.Flap.AnimGraphNode_StateMachine_0.New State Machine.AnimStateTransitionNode_1.Transition
  nodes [Result()]
 }
 Transition {
  path=AnimGraph.AnimGraphNode_StateMachine_4.Locomotion.AnimStateTransitionNode_0.Transition
  nodes [
   Result()
   "Equal (Float)"(A="Get Relevant Anim Time Remaining (Open_Into)"!() -> (ReturnValue:float) {is=title K2Node_AnimGetter}, B=0)
  ]
 }
 Transition {
  path=AnimGraph.AnimGraphNode_StateMachine_4.Locomotion.AnimStateTransitionNode_1.Transition
  nodes [
   n1 = Get SkydivingState()
   Result(bCanEnterTransition=OR Boolean(
     A=Break Fort Anim Input Skydiving(FortAnimInput_Skydiving=@n1) -> (
      ParaGilderRootOffset=~
      ParaGliderRootRotationOverride=~
      bUseParaGlideRootModifier=~
      bIsSkydivingFromLaunchPad=~
      bIsSkydivingFromBus=~
      bIsInVortex=~
      bIsDiving=~
      bIsParachuteOpen=~
      bIsSkydiving=~
      bIsParachuteLeaning=~
      bIsSkydiveLeaning=~
      bIsLeaning=~
      bIsSkydiveDiveMode=~
      bPlaySkydiveDrift=~
      bSkydiveDriftDelayActive=~
      bSkydiveDriftAnimAllowed=~
      bIsGliderRight=~
      bIsGliderCenter=~
      bIsGliderLeft=~
      bIsGliderForward=~
      bIsGliderBack=~
      bHasDeployedAndIsStandingUpright=~
      LocalAccelForward=~
      LocalAccelRight=~
      LocalVelocityRight=~
      SkydiveAimPitch=~
      SkydiveAimYaw=~
      DeployChuteAnimRate=~
      SkydiveDriftAnimRate=~
      SkydiveDriftAnimRateCurrent=~
      SkydiveFidgetAnimRate=~
      SkydiveFidgetAnimRateCurrent=~
      SkydiveAdditiveAlpha=~
      SkydiveDriftDelay=~
     ).bIsParachuteOpen
     B=Get bForceOpen() -> (bForceOpen:bool)
    ))
  ]
 }
 Transition {
  path=AnimGraph.AnimGraphNode_StateMachine_4.Locomotion.AnimStateTransitionNode_2.Transition
  nodes [
   n1 = Get bForceOpen()
   n2 = Break Fort Anim Input Skydiving(FortAnimInput_Skydiving=Get SkydivingState()) -> (
    ParaGilderRootOffset=~
    ParaGliderRootRotationOverride=~
    bUseParaGlideRootModifier=~
    bIsSkydivingFromLaunchPad=~
    bIsSkydivingFromBus=~
    bIsInVortex=~
    bIsDiving=~
    bIsParachuteOpen=~
    bIsSkydiving=~
    bIsParachuteLeaning=~
    bIsSkydiveLeaning=~
    bIsLeaning=~
    bIsSkydiveDiveMode=~
    bPlaySkydiveDrift=~
    bSkydiveDriftDelayActive=~
    bSkydiveDriftAnimAllowed=~
    bIsGliderRight=~
    bIsGliderCenter=~
    bIsGliderLeft=~
    bIsGliderForward=~
    bIsGliderBack=~
    bHasDeployedAndIsStandingUpright=~
    LocalAccelForward=~
    LocalAccelRight=~
    LocalVelocityRight=~
    SkydiveAimPitch=~
    SkydiveAimYaw=~
    DeployChuteAnimRate=~
    SkydiveDriftAnimRate=~
    SkydiveDriftAnimRateCurrent=~
    SkydiveFidgetAnimRate=~
    SkydiveFidgetAnimRateCurrent=~
    SkydiveAdditiveAlpha=~
    SkydiveDriftDelay=~
   )
   Result(bCanEnterTransition=AND Boolean(A=NOT Boolean(A=@n2.bIsParachuteOpen), B=NOT Boolean(A=@n1)))
  ]
 }
 Transition {
  path=AnimGraph.AnimGraphNode_StateMachine_4.Locomotion.AnimStateTransitionNode_3.Transition
  nodes [
   n1 = Get bForceOpen()
   n2 = Break Fort Anim Input Skydiving(FortAnimInput_Skydiving=Get SkydivingState()) -> (
    ParaGilderRootOffset=~
    ParaGliderRootRotationOverride=~
    bUseParaGlideRootModifier=~
    bIsSkydivingFromLaunchPad=~
    bIsSkydivingFromBus=~
    bIsInVortex=~
    bIsDiving=~
    bIsParachuteOpen=~
    bIsSkydiving=~
    bIsParachuteLeaning=~
    bIsSkydiveLeaning=~
    bIsLeaning=~
    bIsSkydiveDiveMode=~
    bPlaySkydiveDrift=~
    bSkydiveDriftDelayActive=~
    bSkydiveDriftAnimAllowed=~
    bIsGliderRight=~
    bIsGliderCenter=~
    bIsGliderLeft=~
    bIsGliderForward=~
    bIsGliderBack=~
    bHasDeployedAndIsStandingUpright=~
    LocalAccelForward=~
    LocalAccelRight=~
    LocalVelocityRight=~
    SkydiveAimPitch=~
    SkydiveAimYaw=~
    DeployChuteAnimRate=~
    SkydiveDriftAnimRate=~
    SkydiveDriftAnimRateCurrent=~
    SkydiveFidgetAnimRate=~
    SkydiveFidgetAnimRateCurrent=~
    SkydiveAdditiveAlpha=~
    SkydiveDriftDelay=~
   )
   Result(bCanEnterTransition=AND Boolean(A=NOT Boolean(A=@n2.bIsParachuteOpen), B=NOT Boolean(A=@n1)))
  ]
 }
 Locomotion_Backup {
  entry=Closed
  path=AnimGraph.AnimGraphNode_StateMachine_5.Locomotion_Backup
  ids=id
  nodes [
   AnimStateNode_0 = Idle_Open()
   AnimStateNode_1 = Open_Into()
   AnimStateNode_2 = Closed()
   AnimStateNode_3 = Flap()
   AnimStateTransitionNode_0 = AnimStateTransitionNode() {bAutomaticRuleBasedOnSequencePlayerInState=true, crossfadeDuration=0.3}
   AnimStateTransitionNode_1 = AnimStateTransitionNode() {crossfadeDuration=0}
   AnimStateTransitionNode_2 = AnimStateTransitionNode() {crossfadeDuration=0}
   AnimStateTransitionNode_3 = AnimStateTransitionNode() {crossfadeDuration=0, priorityOrder=2}
   AnimStateTransitionNode_4 = AnimStateTransitionNode() {crossfadeDuration=0.6}
   AnimStateTransitionNode_5 = AnimStateTransitionNode() {bAutomaticRuleBasedOnSequencePlayerInState=true, crossfadeDuration=0.6}
   AnimStateTransitionNode_6 = AnimStateTransitionNode() {crossfadeDuration=0.3, priorityOrder=2}
   AnimStateTransitionNode_7 = AnimStateTransitionNode() {crossfadeDuration=0}
  ]
  links [
   {
    dst=Open_Into
    rule=AnimGraph.AnimGraphNode_StateMachine_5.Locomotion_Backup.AnimStateTransitionNode_1.Transition
    src=Closed
    via=AnimStateTransitionNode_1
   }
   {
    dst=Closed
    rule=AnimGraph.AnimGraphNode_StateMachine_5.Locomotion_Backup.AnimStateTransitionNode_7.Transition
    src=Flap
    via=AnimStateTransitionNode_7
   }
   {
    dst=Idle_Open
    rule=AnimGraph.AnimGraphNode_StateMachine_5.Locomotion_Backup.AnimStateTransitionNode_5.Transition
    src=Flap
    via=AnimStateTransitionNode_5
   }
   {
    dst=Closed
    rule=AnimGraph.AnimGraphNode_StateMachine_5.Locomotion_Backup.AnimStateTransitionNode_2.Transition
    src=Idle_Open
    via=AnimStateTransitionNode_2
   }
   {
    dst=Flap
    rule=AnimGraph.AnimGraphNode_StateMachine_5.Locomotion_Backup.AnimStateTransitionNode_4.Transition
    src=Idle_Open
    via=AnimStateTransitionNode_4
   }
   {
    dst=Closed
    rule=AnimGraph.AnimGraphNode_StateMachine_5.Locomotion_Backup.AnimStateTransitionNode_3.Transition
    src=Open_Into
    via=AnimStateTransitionNode_3
   }
   {
    dst=Flap
    rule=AnimGraph.AnimGraphNode_StateMachine_5.Locomotion_Backup.AnimStateTransitionNode_6.Transition
    src=Open_Into
    via=AnimStateTransitionNode_6
   }
   {
    dst=Idle_Open
    rule=AnimGraph.AnimGraphNode_StateMachine_5.Locomotion_Backup.AnimStateTransitionNode_0.Transition
    src=Open_Into
    via=AnimStateTransitionNode_0
   }
  ]
 }
 Idle_Open {
  path=AnimGraph.AnimGraphNode_StateMachine_5.Locomotion_Backup.AnimStateNode_0.Idle_Open
  nodes [
   n1 = Break Fort Anim Input Skydiving(FortAnimInput_Skydiving=Get SkydivingState()) -> (
    ParaGilderRootOffset=~
    ParaGliderRootRotationOverride=~
    bUseParaGlideRootModifier=~
    bIsSkydivingFromLaunchPad=~
    bIsSkydivingFromBus=~
    bIsInVortex=~
    bIsDiving=~
    bIsParachuteOpen=~
    bIsSkydiving=~
    bIsParachuteLeaning=~
    bIsSkydiveLeaning=~
    bIsLeaning=~
    bIsSkydiveDiveMode=~
    bPlaySkydiveDrift=~
    bSkydiveDriftDelayActive=~
    bSkydiveDriftAnimAllowed=~
    bIsGliderRight=~
    bIsGliderCenter=~
    bIsGliderLeft=~
    bIsGliderForward=~
    bIsGliderBack=~
    bHasDeployedAndIsStandingUpright=~
    LocalAccelForward=~
    LocalAccelRight=~
    LocalVelocityRight=~
    SkydiveAimPitch=~
    SkydiveAimYaw=~
    DeployChuteAnimRate=~
    SkydiveDriftAnimRate=~
    SkydiveDriftAnimRateCurrent=~
    SkydiveFidgetAnimRate=~
    SkydiveFidgetAnimRateCurrent=~
    SkydiveAdditiveAlpha=~
    SkydiveDriftDelay=~
   )
   n2 = Blendspace Player 'BS_HighBeam_GLIDER'(X=@n1.SkydiveAimYaw, Y=@n1.SkydiveAimPitch) {blendSpace=&BS_HighBeam_GLIDER, groupName=Glide, method=SyncGroup}
   n3 = Get OwnerGlider()
   Output Animation Pose(Result=Blend Poses by bool(
     BlendPose_0=Sequence Player 'HighBeam_FrontEnd_Glide_Anim_GLIDER'() {sequence=&HighBeam_FrontEnd_Glide_Anim_GLIDER}
     BlendPose_1=@n2
     bActiveValue=Get IsCosmeticPreview(self=@n3)
    ) {blendType=Linear})
  ]
 }
 Open_Into {
  path=AnimGraph.AnimGraphNode_StateMachine_5.Locomotion_Backup.AnimStateNode_1.Open_Into
  nodes [
   n1 = Break Fort Anim Input Skydiving(FortAnimInput_Skydiving=Get SkydivingState()) -> (
    ParaGilderRootOffset=~
    ParaGliderRootRotationOverride=~
    bUseParaGlideRootModifier=~
    bIsSkydivingFromLaunchPad=~
    bIsSkydivingFromBus=~
    bIsInVortex=~
    bIsDiving=~
    bIsParachuteOpen=~
    bIsSkydiving=~
    bIsParachuteLeaning=~
    bIsSkydiveLeaning=~
    bIsLeaning=~
    bIsSkydiveDiveMode=~
    bPlaySkydiveDrift=~
    bSkydiveDriftDelayActive=~
    bSkydiveDriftAnimAllowed=~
    bIsGliderRight=~
    bIsGliderCenter=~
    bIsGliderLeft=~
    bIsGliderForward=~
    bIsGliderBack=~
    bHasDeployedAndIsStandingUpright=~
    LocalAccelForward=~
    LocalAccelRight=~
    LocalVelocityRight=~
    SkydiveAimPitch=~
    SkydiveAimYaw=~
    DeployChuteAnimRate=~
    SkydiveDriftAnimRate=~
    SkydiveDriftAnimRateCurrent=~
    SkydiveFidgetAnimRate=~
    SkydiveFidgetAnimRateCurrent=~
    SkydiveAdditiveAlpha=~
    SkydiveDriftDelay=~
   )
   n2 = Get OwnerGlider()
   Output Animation Pose(Result=Apply Additive!(
     Base=Blendspace Player 'BS_HighBeam_Into_GLIDER'(X=@n1.SkydiveAimYaw, Y=@n1.SkydiveAimPitch) {bLoop=false, blendSpace=&BS_HighBeam_Into_GLIDER}
     Additive="Sequence Player (additive) 'HighBeam_FrontEnd_Into_Anim_GLIDER'"() {bLoopAnimation=false, sequence=&HighBeam_FrontEnd_Into_Anim_GLIDER}
     bAlphaBoolEnabled:bool=Get IsCosmeticPreview(self=@n2)
    ) -> (Pose) {alphaInputType=Bool, alphaBoolBlend {blendOption=Linear}})
  ]
 }
 Closed {
  path=AnimGraph.AnimGraphNode_StateMachine_5.Locomotion_Backup.AnimStateNode_2.Closed
  nodes [
   n1 = Get OwnerGlider()
   Output Animation Pose(Result=Blend Poses by bool(
     BlendPose_0=Sequence Player 'HighBeam_Closed_Pose_GLIDER'() {sequence=&HighBeam_Closed_Pose_GLIDER}
     BlendPose_1=Sequence Player 'HighBeam_Closed_Pose_GLIDER'() {sequence=&HighBeam_Closed_Pose_GLIDER}
     bActiveValue=Get IsCosmeticPreview(self=@n1)
    ) {blendType=Linear})
  ]
 }
 Flap {
  path=AnimGraph.AnimGraphNode_StateMachine_5.Locomotion_Backup.AnimStateNode_3.Flap
  nodes [
   "Sequence Player (additive) 'HighBeam_Flap_Anim_GLIDER'"(PlayRate=Get Flap_Rate()) {sequence=&HighBeam_Flap_Anim_GLIDER}
   n1 = Break Fort Anim Input Skydiving(FortAnimInput_Skydiving=Get SkydivingState()) -> (
    ParaGilderRootOffset=~
    ParaGliderRootRotationOverride=~
    bUseParaGlideRootModifier=~
    bIsSkydivingFromLaunchPad=~
    bIsSkydivingFromBus=~
    bIsInVortex=~
    bIsDiving=~
    bIsParachuteOpen=~
    bIsSkydiving=~
    bIsParachuteLeaning=~
    bIsSkydiveLeaning=~
    bIsLeaning=~
    bIsSkydiveDiveMode=~
    bPlaySkydiveDrift=~
    bSkydiveDriftDelayActive=~
    bSkydiveDriftAnimAllowed=~
    bIsGliderRight=~
    bIsGliderCenter=~
    bIsGliderLeft=~
    bIsGliderForward=~
    bIsGliderBack=~
    bHasDeployedAndIsStandingUpright=~
    LocalAccelForward=~
    LocalAccelRight=~
    LocalVelocityRight=~
    SkydiveAimPitch=~
    SkydiveAimYaw=~
    DeployChuteAnimRate=~
    SkydiveDriftAnimRate=~
    SkydiveDriftAnimRateCurrent=~
    SkydiveFidgetAnimRate=~
    SkydiveFidgetAnimRateCurrent=~
    SkydiveAdditiveAlpha=~
    SkydiveDriftDelay=~
   )
   n2 = Blendspace Player 'BS_HighBeam_GLIDER'(X=@n1.SkydiveAimYaw, Y=@n1.SkydiveAimPitch) {
    bIgnoreForRelevancyTest=true
    blendSpace=&BS_HighBeam_GLIDER
    groupName=Glide
    method=SyncGroup
   }
   n3 = Sequence Player 'HighBeam_FrontEnd_Glide_Anim_GLIDER'() {bIgnoreForRelevancyTest=true, sequence=&HighBeam_FrontEnd_Glide_Anim_GLIDER}
   n4 = Get IsCosmeticPreview(self=Get OwnerGlider())
   AnimGraphNode_StateMachine_0 = New State Machine()
   Output Animation Pose(Result=Apply Additive!(
     Base=Blend Poses by bool(BlendPose_0=@n3, BlendPose_1=@n2, bActiveValue=@n4) {blendType=Linear}
     Additive=@AnimGraphNode_StateMachine_0
     Alpha:float
    ) -> (Pose) {alphaBoolBlend {blendOption=Linear}})
   n5 = GetCurveValue(CurveName=DisableFlap)
   AND Boolean(A=OR Boolean(A=Get CanFlap(), B=Get TransitionFlap()), B=NOT Boolean(A="Equal (Float)"(A=@n5, B=1)))
  ]
 }
 New State Machine {
  entry=Identity
  path=AnimGraph.AnimGraphNode_StateMachine_5.Locomotion_Backup.AnimStateNode_3.Flap.AnimGraphNode_StateMachine_0.New State Machine
  ids=id
  nodes [
   AnimStateNode_0 = Identity()
   AnimStateNode_1 = Flap() {
    bAlwaysResetOnEntry=true
    stateEntered {guid=#1, notifyName=EnteredFlap_Backup}
    stateLeft {guid=#2, notifyName=ExitedFlap_Backup}
   }
   AnimStateTransitionNode_0 = AnimStateTransitionNode() {crossfadeDuration=0.6}
   AnimStateTransitionNode_1 = AnimStateTransitionNode() {bAutomaticRuleBasedOnSequencePlayerInState=true, crossfadeDuration=0.6}
  ]
  links [
   {
    dst=Identity
    rule=AnimGraph.AnimGraphNode_StateMachine_5.Locomotion_Backup.AnimStateNode_3.Flap.AnimGraphNode_StateMachine_0.New State Machine.AnimStateTransitionNode_1.Transition
    src=Flap
    via=AnimStateTransitionNode_1
   }
   {
    dst=Flap
    rule=AnimGraph.AnimGraphNode_StateMachine_5.Locomotion_Backup.AnimStateNode_3.Flap.AnimGraphNode_StateMachine_0.New State Machine.AnimStateTransitionNode_0.Transition
    src=Identity
    via=AnimStateTransitionNode_0
   }
  ]
 }
 Identity {
  path=AnimGraph.AnimGraphNode_StateMachine_5.Locomotion_Backup.AnimStateNode_3.Flap.AnimGraphNode_StateMachine_0.New State Machine.AnimStateNode_0.Identity
  nodes [Output Animation Pose(Result=Additive Identity Pose())]
 }
 Flap {
  path=AnimGraph.AnimGraphNode_StateMachine_5.Locomotion_Backup.AnimStateNode_3.Flap.AnimGraphNode_StateMachine_0.New State Machine.AnimStateNode_1.Flap
  nodes [
   Output Animation Pose(Result="Sequence Player (additive) 'HighBeam_Flap_Anim_GLIDER'"(PlayRate=1) {bLoopAnimation=false, sequence=&HighBeam_Flap_Anim_GLIDER})
   Get Flap_Rate()
  ]
 }
 Transition {
  path=AnimGraph.AnimGraphNode_StateMachine_5.Locomotion_Backup.AnimStateNode_3.Flap.AnimGraphNode_StateMachine_0.New State Machine.AnimStateTransitionNode_0.Transition
  nodes [
   n1 = Get CanFlap()
   Result(bCanEnterTransition=@n1)
   n2 = GetCurveValue(CurveName=DisableFlap)
   AND Boolean(A=OR Boolean(A=@n1, B=Get TransitionFlap()), B=NOT Boolean(A="Equal (Float)"(A=@n2, B=1)))
  ]
 }
 Transition {
  path=AnimGraph.AnimGraphNode_StateMachine_5.Locomotion_Backup.AnimStateNode_3.Flap.AnimGraphNode_StateMachine_0.New State Machine.AnimStateTransitionNode_1.Transition
  nodes [Result()]
 }
 Transition {
  path=AnimGraph.AnimGraphNode_StateMachine_5.Locomotion_Backup.AnimStateTransitionNode_0.Transition
  nodes [
   Result()
   "Equal (Float)"(A="Get Relevant Anim Time Remaining (Open_Into)"(), B=0)
  ]
 }
 Transition {
  path=AnimGraph.AnimGraphNode_StateMachine_5.Locomotion_Backup.AnimStateTransitionNode_1.Transition
  nodes [
   n1 = Get SkydivingState()
   Result(bCanEnterTransition=OR Boolean(
     A=Break Fort Anim Input Skydiving(FortAnimInput_Skydiving=@n1) -> (
      ParaGilderRootOffset=~
      ParaGliderRootRotationOverride=~
      bUseParaGlideRootModifier=~
      bIsSkydivingFromLaunchPad=~
      bIsSkydivingFromBus=~
      bIsInVortex=~
      bIsDiving=~
      bIsParachuteOpen=~
      bIsSkydiving=~
      bIsParachuteLeaning=~
      bIsSkydiveLeaning=~
      bIsLeaning=~
      bIsSkydiveDiveMode=~
      bPlaySkydiveDrift=~
      bSkydiveDriftDelayActive=~
      bSkydiveDriftAnimAllowed=~
      bIsGliderRight=~
      bIsGliderCenter=~
      bIsGliderLeft=~
      bIsGliderForward=~
      bIsGliderBack=~
      bHasDeployedAndIsStandingUpright=~
      LocalAccelForward=~
      LocalAccelRight=~
      LocalVelocityRight=~
      SkydiveAimPitch=~
      SkydiveAimYaw=~
      DeployChuteAnimRate=~
      SkydiveDriftAnimRate=~
      SkydiveDriftAnimRateCurrent=~
      SkydiveFidgetAnimRate=~
      SkydiveFidgetAnimRateCurrent=~
      SkydiveAdditiveAlpha=~
      SkydiveDriftDelay=~
     ).bIsParachuteOpen
     B=Get bForceOpen()
    ))
  ]
 }
 Transition {
  path=AnimGraph.AnimGraphNode_StateMachine_5.Locomotion_Backup.AnimStateTransitionNode_2.Transition
  nodes [
   n1 = Get bForceOpen()
   n2 = Break Fort Anim Input Skydiving(FortAnimInput_Skydiving=Get SkydivingState()) -> (
    ParaGilderRootOffset=~
    ParaGliderRootRotationOverride=~
    bUseParaGlideRootModifier=~
    bIsSkydivingFromLaunchPad=~
    bIsSkydivingFromBus=~
    bIsInVortex=~
    bIsDiving=~
    bIsParachuteOpen=~
    bIsSkydiving=~
    bIsParachuteLeaning=~
    bIsSkydiveLeaning=~
    bIsLeaning=~
    bIsSkydiveDiveMode=~
    bPlaySkydiveDrift=~
    bSkydiveDriftDelayActive=~
    bSkydiveDriftAnimAllowed=~
    bIsGliderRight=~
    bIsGliderCenter=~
    bIsGliderLeft=~
    bIsGliderForward=~
    bIsGliderBack=~
    bHasDeployedAndIsStandingUpright=~
    LocalAccelForward=~
    LocalAccelRight=~
    LocalVelocityRight=~
    SkydiveAimPitch=~
    SkydiveAimYaw=~
    DeployChuteAnimRate=~
    SkydiveDriftAnimRate=~
    SkydiveDriftAnimRateCurrent=~
    SkydiveFidgetAnimRate=~
    SkydiveFidgetAnimRateCurrent=~
    SkydiveAdditiveAlpha=~
    SkydiveDriftDelay=~
   )
   Result(bCanEnterTransition=AND Boolean(A=NOT Boolean(A=@n2.bIsParachuteOpen), B=NOT Boolean(A=@n1)))
  ]
 }
 Transition {
  path=AnimGraph.AnimGraphNode_StateMachine_5.Locomotion_Backup.AnimStateTransitionNode_3.Transition
  nodes [
   n1 = Get bForceOpen()
   n2 = Break Fort Anim Input Skydiving(FortAnimInput_Skydiving=Get SkydivingState()) -> (
    ParaGilderRootOffset=~
    ParaGliderRootRotationOverride=~
    bUseParaGlideRootModifier=~
    bIsSkydivingFromLaunchPad=~
    bIsSkydivingFromBus=~
    bIsInVortex=~
    bIsDiving=~
    bIsParachuteOpen=~
    bIsSkydiving=~
    bIsParachuteLeaning=~
    bIsSkydiveLeaning=~
    bIsLeaning=~
    bIsSkydiveDiveMode=~
    bPlaySkydiveDrift=~
    bSkydiveDriftDelayActive=~
    bSkydiveDriftAnimAllowed=~
    bIsGliderRight=~
    bIsGliderCenter=~
    bIsGliderLeft=~
    bIsGliderForward=~
    bIsGliderBack=~
    bHasDeployedAndIsStandingUpright=~
    LocalAccelForward=~
    LocalAccelRight=~
    LocalVelocityRight=~
    SkydiveAimPitch=~
    SkydiveAimYaw=~
    DeployChuteAnimRate=~
    SkydiveDriftAnimRate=~
    SkydiveDriftAnimRateCurrent=~
    SkydiveFidgetAnimRate=~
    SkydiveFidgetAnimRateCurrent=~
    SkydiveAdditiveAlpha=~
    SkydiveDriftDelay=~
   )
   Result(bCanEnterTransition=AND Boolean(A=NOT Boolean(A=@n2.bIsParachuteOpen), B=NOT Boolean(A=@n1)))
  ]
 }
 Transition {
  path=AnimGraph.AnimGraphNode_StateMachine_5.Locomotion_Backup.AnimStateTransitionNode_4.Transition
  nodes [
   n1 = "Equal (Float)"(A=GetCurveValue(CurveName=DisableFlap), B=1)
   n2 = Get CanFlap()
   n3 = Get TransitionFlap()
   Result(bCanEnterTransition=AND Boolean(A=OR Boolean(A=@n2, B=@n3), B=NOT Boolean(A=@n1)))
  ]
 }
 Transition {path=AnimGraph.AnimGraphNode_StateMachine_5.Locomotion_Backup.AnimStateTransitionNode_5.Transition, nodes [Result()]}
 Transition {
  path=AnimGraph.AnimGraphNode_StateMachine_5.Locomotion_Backup.AnimStateTransitionNode_6.Transition
  nodes [
   n1 = "Equal (Float)"(A=GetCurveValue(CurveName=DisableFlap), B=1)
   n2 = Get CanFlap()
   n3 = Get TransitionFlap()
   Result(bCanEnterTransition=AND Boolean(A=OR Boolean(A=@n2, B=@n3), B=NOT Boolean(A=@n1)))
  ]
 }
 Transition {
  path=AnimGraph.AnimGraphNode_StateMachine_5.Locomotion_Backup.AnimStateTransitionNode_7.Transition
  nodes [
   n1 = Get bForceOpen()
   n2 = Break Fort Anim Input Skydiving(FortAnimInput_Skydiving=Get SkydivingState()) -> (
    ParaGilderRootOffset=~
    ParaGliderRootRotationOverride=~
    bUseParaGlideRootModifier=~
    bIsSkydivingFromLaunchPad=~
    bIsSkydivingFromBus=~
    bIsInVortex=~
    bIsDiving=~
    bIsParachuteOpen=~
    bIsSkydiving=~
    bIsParachuteLeaning=~
    bIsSkydiveLeaning=~
    bIsLeaning=~
    bIsSkydiveDiveMode=~
    bPlaySkydiveDrift=~
    bSkydiveDriftDelayActive=~
    bSkydiveDriftAnimAllowed=~
    bIsGliderRight=~
    bIsGliderCenter=~
    bIsGliderLeft=~
    bIsGliderForward=~
    bIsGliderBack=~
    bHasDeployedAndIsStandingUpright=~
    LocalAccelForward=~
    LocalAccelRight=~
    LocalVelocityRight=~
    SkydiveAimPitch=~
    SkydiveAimYaw=~
    DeployChuteAnimRate=~
    SkydiveDriftAnimRate=~
    SkydiveDriftAnimRateCurrent=~
    SkydiveFidgetAnimRate=~
    SkydiveFidgetAnimRateCurrent=~
    SkydiveAdditiveAlpha=~
    SkydiveDriftDelay=~
   )
   Result(bCanEnterTransition=AND Boolean(A=NOT Boolean(A=@n2.bIsParachuteOpen), B=NOT Boolean(A=@n1)))
  ]
 }
 EventGraph {nodes [
   n1 = RandomFloatInRange(Min=4, Max=8)
   n2 = Delay(Duration=@n1)
   Set canFlap(execute=[@n2.then, @n5.then], canFlap:bool=true) -> (Output_Get:bool)
   ForceFlapOnLean() {is=title K2Node_CustomEvent}
   > n3 = Branch(Condition=Get Flapping() -> (Flapping:bool))
   n4 = Set canFlap(execute=@n3.else, canFlap=true)
   n5 = RetriggerableDelay(execute=[@n4.then, @n14.then], Duration=@n1)
   Event BlueprintUpdateAnimation()
   > Sequence!(execute) -> (then_0, then_1, then_2:Exec, then_3:Exec)
   n6 = Break Fort Anim Input Skydiving(FortAnimInput_Skydiving=Get SkydivingState()) -> (
    ParaGilderRootOffset=~
    ParaGliderRootRotationOverride=~
    bUseParaGlideRootModifier=~
    bIsSkydivingFromLaunchPad=~
    bIsSkydivingFromBus=~
    bIsInVortex=~
    bIsDiving=~
    bIsParachuteOpen=~
    bIsSkydiving=~
    bIsParachuteLeaning=~
    bIsSkydiveLeaning=~
    bIsLeaning=~
    bIsSkydiveDiveMode=~
    bPlaySkydiveDrift=~
    bSkydiveDriftDelayActive=~
    bSkydiveDriftAnimAllowed=~
    bIsGliderRight=~
    bIsGliderCenter=~
    bIsGliderLeft=~
    bIsGliderForward=~
    bIsGliderBack=~
    bHasDeployedAndIsStandingUpright=~
    LocalAccelForward=~
    LocalAccelRight=~
    LocalVelocityRight=~
    SkydiveAimPitch=~
    SkydiveAimYaw=~
    DeployChuteAnimRate=~
    SkydiveDriftAnimRate=~
    SkydiveDriftAnimRateCurrent=~
    SkydiveFidgetAnimRate=~
    SkydiveFidgetAnimRateCurrent=~
    SkydiveAdditiveAlpha=~
    SkydiveDriftDelay=~
   )
   > n7 = Branch(Condition=OR Boolean!(A=@n6.bIsGliderRight, B=@n6.bIsGliderLeft, C:bool=@n6.bIsGliderForward, D:bool=@n6.bIsGliderBack) -> (ReturnValue))
   > Do Once(Reset=@n7.else)
   > ForceFlapOnLean() {is=title K2Node_CallFunction}
   AnimNotify_TurnONEngines() {is=title K2Node_Event}
   > SetGliderFullyDeployed!(execute, self:Fort Glider Anim Instance Object Reference, bIsFullyDeployed:bool=true) -> (then)
   AnimNotify_ExitedBlinkWithEyeDart() {is=title K2Node_Event}
   > n8 = Delay(Duration=RandomFloatInRange(Min=2, Max=8))
   n9 = Branch(execute=[@n8.then, @n12.then, @n18.then], Condition=RandomBool!() -> (ReturnValue:bool))
   > Set CanBlink(CanBlink:bool=true) -> (Output_Get:bool)
   Set CanBlinkWithEyeDart(execute=@n9.else, CanBlinkWithEyeDart:bool=true) -> (Output_Get:bool)
   AnimNotify_ExitedBlinkIdentity() {is=title K2Node_Event}
   > Set CanBlink(CanBlink=false)
   > Set CanBlinkWithEyeDart(CanBlinkWithEyeDart=false)
   AnimNotify_ExitedCawIdentity() {is=title K2Node_Event}
   > Set CanCaw(CanCaw:bool=false) -> (Output_Get:bool)
   AnimNotify_FlapEnable() {is=title K2Node_Event}
   > Set CanFlap(CanFlap:bool=true) -> (Output_Get:bool)
   > n10 = Sequence!(execute) -> (then_0, then_1)
   > Delay(Duration=2)
   > n11 = Branch(Condition=RandomBool())
   > Set CanLookLeft(CanLookLeft:bool=true) -> (Output_Get:bool)
   Set CanLookRight(execute=@n11.else, CanLookRight:bool=true) -> (Output_Get:bool)
   n12 = Delay(execute=@n10.then_1, Duration=RandomFloatInRange(Min=0, Max=6))
   AnimNotify_EnteredFlap() {is=title K2Node_Event}
   > n13 = Set Flapping(Flapping:bool=true) -> (Output_Get:bool)
   > Call Flapping State Toggled!(execute:Exec, self:High Beam Glider New Anim BP Object Reference, Flapping:bool=@n13) -> (then:Exec) {is=title K2Node_CallDelegate}
   > n14 = Set canFlap(canFlap=false)
   AnimNotify_Entered Look Left() {is=title K2Node_Event}
   > n15 = Set CanLookLeft(CanLookLeft=false)
   Delay(execute=[@n17.then, @n15.then], Duration=RandomFloatInRange(Min=4, Max=10))
   > n16 = Branch(Condition=RandomBool())
   > Set CanLookLeft(CanLookLeft=true)
   Set CanLookRight(execute=@n16.else, CanLookRight=true)
   AnimNotify_Entered Look Right() {is=title K2Node_Event}
   > n17 = Set CanLookRight(CanLookRight=false)
   AnimNotify_ExitedBlink() {is=title K2Node_Event}
   > n18 = Delay(Duration=RandomFloatInRange(Min=2, Max=8))
   AnimNotify_ExitedFlap() {is=title K2Node_Event}
   > n19 = Set Flapping(Flapping=false)
   > Call Flapping State Toggled(Flapping=@n19)
   > Branch(Condition=RandomBool())
   > Set CanCaw(CanCaw=true)
   n20 = Branch()
   > Set BlockFlapsAndLooks(BlockFlapsAndLooks:bool=true) -> (Output_Get:bool)
   Set BlockFlapsAndLooks(execute=@n20.else, BlockFlapsAndLooks=false)
   Break Fort Anim Input Skydiving(FortAnimInput_Skydiving=Get SkydivingState()) -> (
    ParaGilderRootOffset=~
    ParaGliderRootRotationOverride=~
    bUseParaGlideRootModifier=~
    bIsSkydivingFromLaunchPad=~
    bIsSkydivingFromBus=~
    bIsInVortex=~
    bIsDiving=~
    bIsParachuteOpen=~
    bIsSkydiving=~
    bIsParachuteLeaning=~
    bIsSkydiveLeaning=~
    bIsLeaning=~
    bIsSkydiveDiveMode=~
    bPlaySkydiveDrift=~
    bSkydiveDriftDelayActive=~
    bSkydiveDriftAnimAllowed=~
    bIsGliderRight=~
    bIsGliderCenter=~
    bIsGliderLeft=~
    bIsGliderForward=~
    bIsGliderBack=~
    bHasDeployedAndIsStandingUpright=~
    LocalAccelForward=~
    LocalAccelRight=~
    LocalVelocityRight=~
    SkydiveAimPitch=~
    SkydiveAimYaw=~
    DeployChuteAnimRate=~
    SkydiveDriftAnimRate=~
    SkydiveDriftAnimRateCurrent=~
    SkydiveFidgetAnimRate=~
    SkydiveFidgetAnimRateCurrent=~
    SkydiveAdditiveAlpha=~
    SkydiveDriftDelay=~
   )
   AnimNotify_FlapOff() {is=title K2Node_Event}
   AnimNotify_LookOff() {is=title K2Node_Event}
  ]}
 Flapping State Toggled {nodes [Flapping State Toggled!() -> (then, Flapping:bool=~) {is=title K2Node_FunctionEntry}]}
]
variables [
 BlockFlapsAndLooks {type=boolean}
 CanBlink {type=boolean}
 CanBlinkWithEyeDart {type=boolean}
 CanCaw {type=boolean}
 CanFlap {type=boolean}
 CanLookLeft {type=boolean}
 CanLookRight {type=boolean}
 Flap_Delay {type=number}
 Flap_Rate {type=number}
 FlapAnim {type=AnimSequence, default=null}
 Flapping {type=boolean}
 Flapping State Toggled {default=()}
 TransitionFlap {type=boolean}
 useLeanTransition {type=boolean}
]
